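# Physics asset for the female holiday Renegade Raider pickaxe, giving its bow and cloth bones collision bodies and joint limits for secondary motion.
# A long capsule covers the handle, with capsules on the bow center and loops and boxes on the bow tails and cloth.
# Two loops and two two-segment tail chains hang from dyn_bow with tight limits; a looser two-segment cloth chain hangs from the handle.
# No constraint links dyn_bow to the handle.
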PhysicsAsset /BRCosmetics/Weapons/FORT_Melee/Pickaxe_Renegade_Raider_Holiday_Female/Meshes/Renegade_Raider_Holiday_Female_Axe_Physics
bodies {
 Melee_TwoHanded_Handle [{shapeType=Capsule, length=118.8762, radius=2.3297, center=<0.7582 -1.1442 15.0458>, rotation=<rot -0.1691 -89.9646 0.2896>}]
 dyn_bow [{shapeType=Capsule, length=0.0099, radius=2.1214, center=<-1.3508 -0.2803 -0.0255>, rotation=<rot 0 0.214 -90.8195>}]
 dyn_bow_l [{shapeType=Capsule, length=7.5275, radius=2.1161, center=<-1.3578 1.2038 -0.8037>, rotation=<rot -76.7164 -83.5121 -6.3297>}]
 dyn_bow_l_01 [{shapeType=Box, extentX=1.0221, extentY=3.5165, extentZ=5.9238, center=<0 0 -1.9508>}]
 dyn_bow_l_02 [{shapeType=Box, extentX=1.0221, extentY=5.051, extentZ=5.9238, center=<0 0 -1.9508>}]
 dyn_bow_r [{shapeType=Capsule, length=7.097, radius=2.1448, center=<-1.2932 -1.7672 -1.3891>, rotation=<rot 67.2924 -94.4774 -4.4717>}]
 dyn_bow_r_01 [{shapeType=Box, extentX=1.0221, extentY=3.5165, extentZ=5.9238, center=<0 0 -1.9508>}]
 dyn_bow_r_02 [{shapeType=Box, extentX=1.0221, extentY=5.051, extentZ=5.9238, center=<0 0 -1.9508>}]
 dyn_cloth01 [{shapeType=Box, extentX=1.2106, extentY=5.889, extentZ=2.7386, center=<0.0001 1.1878 -0.2568>, rotation=<rot 0 0 12.2014>}]
 dyn_cloth02 [{shapeType=Box, extentX=1.2106, extentY=5.889, extentZ=2.7386, center=<0.0001 1.1922 -0.2356>, rotation=<rot 0 0 11.1815>}]
}
constraints [
 {bone1Name=dyn_bow_l_01, bone2Name=dyn_bow, swing1LimitDegrees=10, swing2LimitDegrees=30, twistLimitDegrees=30}
 {bone1Name=dyn_bow_l_02, bone2Name=dyn_bow_l_01, swing1LimitDegrees=10, swing2LimitDegrees=20, twistLimitDegrees=20}
 {bone1Name=dyn_bow_r_01, bone2Name=dyn_bow, swing1LimitDegrees=10, swing2LimitDegrees=30, twistLimitDegrees=30}
 {bone1Name=dyn_bow_r_02, bone2Name=dyn_bow_r_01, swing1LimitDegrees=10, swing2LimitDegrees=20, twistLimitDegrees=20}
 {bone1Name=dyn_cloth01, bone2Name=Melee_TwoHanded_Handle, swing1LimitDegrees=90, swing2LimitDegrees=10, twistLimitDegrees=90}
 {bone1Name=dyn_cloth02, bone2Name=dyn_cloth01, swing1LimitDegrees=90, swing2LimitDegrees=10, twistLimitDegrees=90}
 {bone1Name=dyn_bow_r, bone2Name=dyn_bow, swing1LimitDegrees=30, swing2LimitDegrees=10, twistLimitDegrees=10}
 {bone1Name=dyn_bow_l, bone2Name=dyn_bow, swing1LimitDegrees=30, swing2LimitDegrees=10, twistLimitDegrees=10}
]
